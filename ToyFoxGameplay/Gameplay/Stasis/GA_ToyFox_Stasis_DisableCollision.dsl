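; Gameplay ability for ToyFox Stasis that turns off collision on the avatar player pawn while the ability is active.
; On activation it caches the valid Fort player pawn, disables its actor collision and adds a configurable gameplay cue to the owner.
; The graph never ends the ability itself.
; On end, cancelled or not, it re-enables collision if the cached pawn is valid, then calls the parent OnEndAbility.
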(event Ability|EventActivateAbility
  ;[0 0.15 0.01041 0.5] ActivateAbility: Disable collision on the player pawn.
  (bind _outfortplayerpawn (CallFunction|GetAvatarasValidFortPlayerPawn)
    (:ValidFortPlayerPawn
      (bind _output_get (Variables|Transient|SetAvatarPlayerPawn _outfortplayerpawn))
      (Collision|SetActorEnableCollision _output_get false)
      (Ability|AddGameplayCueToOwner (Utilities|Struct|BreakGameplayCueTag (Variables|Default|GetGameplay%20Cue%20Tag))))))

(event Ability|EventOnEndAbility
  (params (bWasCancelled :type Boolean))
  ;[0.1146 0.04993 0 0.5] OnEndAbility: Renable collision on the player pawn.
  (Utilities|FlowControl|Sequence
    (:then_0
      (bind _avatarplayerpawn (Variables|Transient|GetAvatarPlayerPawn)
        (:then
          (Collision|SetActorEnableCollision _avatarplayerpawn true))
        (:else)))
    (:then_1
      (|CallParentFunction bWasCancelled))))
