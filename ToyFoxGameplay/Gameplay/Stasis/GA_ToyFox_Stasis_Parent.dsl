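; Parent gameplay ability that puts a ToyFox player pawn into stasis and releases it when the ability ends.
; Activation drops held objects, exits special movement modes, applies damage immunity, holsters weapons, revives from DBNO and cancels building.
; It optionally hides the pawn and disables collision, then adds NoMovementOrFalling stasis next tick and calls the empty HandleStasisSetComplete hook.
; Ending reverses immunity, hiding, holstering and stasis; hiding and holstering are authority-only.
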(event Ability|EventActivateAbility
  ;[0.02954 0.1979 0 0.5] ActivateAbility: The character is put into stasis mode.
  (bind _outfortplayerpawn (CallFunction|GetAvatarasValidFortPlayerPawn)
    (:ValidFortPlayerPawn
      (bind _output_get (Variables|Transient|SetPlayerPawn _outfortplayerpawn))
      (bind _returnvalue (Fortnite|GetFortPlayerControllerfromActor _output_get))
      (Variables|Transient|SetPlayerController _returnvalue)
      (CallFunction|DropHeldObject)
      (Default|EndMovementModes)
      (bind _returnvalue_2 (Ability|ApplyGameplayEffectToOwner (Variables|Settings|GetDamageImmunityGameplayEffect)))
      (Variables|Transient|SetDamageImmunityGameplayEffectHandle _returnvalue_2)
      (CallFunction|SetWeaponHolster true)
      (CallFunction|RevivefromDBNO)
      (CallFunction|ForceCancelBuildingTools)
      (if (Variables|Settings|AutoApply|GetbSetPawnHiddenAndDisableCollision)
        (CallFunction|SetPawnHiddenandDisableCollision true)
        (else
          _))
      (Utilities|FlowControl|DelayUntilNextTick)
      (if (Ability|GameplayAbility|IsActive 0)
        (bind _playerpawn (Variables|Transient|GetPlayerPawn)
          (:then
            (Stasis|AddStasisMode _playerpawn "NoMovementOrFalling" 0 (Variables|Settings|GetStasisGameplayTag) (Variables|Settings|GetStasisAllowedTags))
            (CallFunction|HandleStasisSetComplete))
          (:else))))))

(event Ability|EventOnEndAbility
  (params (bWasCancelled :type Boolean))
  ;[0.09375 0.0289 0 0.5] OnEndAbility: Stasis is removed from the character.
  (Utilities|FlowControl|Sequence
    (:then_0
      (Ability|RemoveGameplayEffectFromOwnerWithHandle (Variables|Transient|GetDamageImmunityGameplayEffectHandle) 1)
      (CallFunction|SetPawnHiddenandDisableCollision false)
      (CallFunction|SetWeaponHolster false)
      (bind _playerpawn (Variables|Transient|GetPlayerPawn)
        (:then
          (Stasis|RemoveStasisMode _playerpawn (Variables|Settings|GetStasisGameplayTag) false))
        (:else)))
    (:then_1
      (|CallParentFunction bWasCancelled))))

(fn DropHeldObject
  (bind _playerpawn (Variables|Transient|GetPlayerPawn)
    (:then
      (bind _returnvalue (Fortnite|HeldObjects|GetHeldObjectComponentAttachedtoPlayer _playerpawn))
      (Utilities|IsValid _returnvalue
        (:Is%20Valid
          (if (Vehicle|IsinVehicle _playerpawn)
            (HeldObjects|Behavior|DropObjectHeldinVehicle _returnvalue)
            (else
              (HeldObjects|Behavior|DropHeldObject _returnvalue))))))
    (:else)))

(fn EndMovement_Ziplining
  (params (PlayerPawn :type Fort%20Player%20Pawn%20Object%20Reference))
  ;[0.00748 0.1198 0.1186] EndMovement_Ziplining: End the zipline movement on the player.
  (Utilities|IsValid PlayerPawn
    (:Is%20Valid
      (Ability|HasAuthority
        (:True
          (bind _returnvalue (Zipline|GetActiveZipline PlayerPawn))
          (Utilities|IsValid _returnvalue
            (:Is%20Valid
              (Class|FortAthenaZiplineBase|AuthorityForceEndZiplining _returnvalue PlayerPawn))))
        (:False
          (Zipline|EndZiplining PlayerPawn false false))))))

(fn EndMovementModes
  (bind _playerpawn (Variables|Transient|GetPlayerPawn))
  (bind _returnvalue (Vehicle|GetVehicleActor _playerpawn :!class /Script/FortniteGame.FortPlayerPawn))
  (bind _playerpawn_2 (Variables|Transient|GetPlayerPawn)
    (:then
      (Utilities|FlowControl|Sequence
        (:then_0
          (Ability|SendGameplayEventtoActor _playerpawn_2 "(TagName=\"Event.Athena.Balloons.RemoveAll\")") ; Remove any balloons on the player.
          ;[0.06532 0.01858 0.1302] End vehicle driving and passenger movement modes.
          (if (GameplayTags|HasAnyTags (GameplayTags|GetOwnedGameplayTags _playerpawn_2 :!class /Script/GameplayTags.BlueprintGameplayTagLibrary) (GameplayTags|MakeLiteralGameplayTagContainer "(GameplayTags=((TagName=\"MovementMode.Driving\"),(TagName=\"MovementMode.Passenger\")))") true)
            (Utilities|IsValid _returnvalue
              (:Is%20Valid
                (bind _asfort_athena_vehicle (Utilities|Casting|CastToFortAthenaVehicle _returnvalue))
                (LocalVariables|SetbSkydiveOnExit (Variables|/Script/FortniteGame.FortAthenaVehicle|GetbSkydivePlayerOnVehicleExit _asfort_athena_vehicle))
                (Variables|/Script/FortniteGame.FortAthenaVehicle|SetbSkydivePlayerOnVehicleExit false _asfort_athena_vehicle)
                (Vehicle|EjectPlayerPawn(Message) _asfort_athena_vehicle _playerpawn)
                (Variables|/Script/FortniteGame.FortAthenaVehicle|SetbSkydivePlayerOnVehicleExit (LocalVariables|GetbSkydiveOnExit) _asfort_athena_vehicle)))
            (elif (GameplayTags|HasAnyTags (GameplayTags|GetOwnedGameplayTags _playerpawn_2 :!class /Script/GameplayTags.BlueprintGameplayTagLibrary) (GameplayTags|MakeLiteralGameplayTagContainer "(GameplayTags=((TagName=\"MovementMode.Parachuting\"),(TagName=\"MovementMode.Skydiving\"),(TagName=\"MovementMode.Skydiving.Following\")))") true)
              ;[0.02695 0.06967 0.151] End skydiving.
              (Parachute|EndSkydiving (Variables|Transient|GetPlayerPawn))
              (elif (GameplayTags|HasAnyTags (GameplayTags|GetOwnedGameplayTags _playerpawn_2 :!class /Script/GameplayTags.BlueprintGameplayTagLibrary) (GameplayTags|MakeLiteralGameplayTagContainer "(GameplayTags=((TagName=\"MovementMode.Ziplining\"),(TagName=\"MovementMode.ZipliningOnSpline\")))") true)
                ;[0.05208 0.008721 0.05085] End ziplining.
                (Default|EndMovementZiplining (Variables|Transient|GetPlayerPawn))
                (elif (GameplayTags|HasAnyTags (GameplayTags|GetOwnedGameplayTags _playerpawn_2 :!class /Script/GameplayTags.BlueprintGameplayTagLibrary) (GameplayTags|MakeLiteralGameplayTagContainer "(GameplayTags=((TagName=\"MovementMode.Swinging\")))") true)
                  ;[0.03744 0.1094 0.07412] End swinging.
                  (Ability|SendGameplayEventtoActor (Variables|Transient|GetPlayerPawn) "(TagName=\"Event.Athena.StopSwinging\")"))))))
        (:then_1
          (bind _charactermovement (Variables|/Script/Engine.Character|GetCharacterMovement _playerpawn_2)
            (:then
              (Pawn|Components|CharacterMovement|DisableMovement _charactermovement) ; Disables MMEs.
            )
            (:else)))))
    (:else)))

(fn ForceCancelBuildingTools
  (bind _playerpawn (Variables|Transient|GetPlayerPawn)
    (:then
      (bind _returnvalue (Fortnite|GetFortPlayerControllerfromActor _playerpawn))
      (Utilities|IsValid _returnvalue
        (:Is%20Valid
          (Building|ForceCancelBuildingTool _returnvalue))))
    (:else)))

(fn HandleStasisSetComplete)

(fn CanActivateAbility
  (params (ActorInfo :type Gameplay%20Ability%20Actor%20Info%20Structure) (Handle :type Gameplay%20Ability%20Spec%20Handle%20Structure))
  (return 0 true))

(fn ReviveFromDBNO
  (bind _playerpawn (Variables|Transient|GetPlayerPawn)
    (:then
      (FortPlayerPawn|RevivefromDBNO _playerpawn (Fortnite|GetFortPlayerControllerfromActor _playerpawn)))
    (:else)))

(fn SetPawnHiddenAndDisableCollision
  (params (bNewHidden :type Boolean))
  (Ability|HasAuthority
    (:True
      (Utilities|FlowControl|DeltaGateWildcard (Variables|Transient|GetbPawnHiddenSet) bNewHidden
        (:Success
          (bind _playerpawn (Variables|Transient|GetPlayerPawn)
            (:then
              (Rendering|SetActorHiddenInGameWithTag _playerpawn bNewHidden (Variables|Settings|GetHidePawnGameplayTag) 0 true)
              (if bNewHidden
                (bind _returnvalue (Ability|ApplyGameplayEffectToOwner (Variables|Settings|GetDisableCollisionGameplayEffectClass)))
                (Variables|Transient|SetDisableCollisionGameplayEffectHandle _returnvalue)
                (else
                  (Ability|RemoveGameplayEffectFromOwnerWithHandle (Variables|Transient|GetDisableCollisionGameplayEffectHandle) 1))))
            (:else)))))))

(fn SetWeaponHolster
  (params (bHolster :type Boolean))
  (Ability|HasAuthority
    (:True
      (bind _playerpawn (Variables|Transient|GetPlayerPawn)
        (:then
          (Weapon|SetHolsterWeapon _playerpawn self bHolster false "ToyFoxPilotTeleport"))
        (:else)))))
