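; GA_ToyFox_Stasis_Pilot is a gameplay ability, apparently for the pilot role in the ToyFox stasis feature.
; On activation it runs the parent ability logic, then resolves the avatar as a valid Fort player pawn.
; If that pawn is locally controlled, it has the pawn's FortPlayerController cache the items currently slotted in the primary bar.
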(event Ability|EventActivateAbility
  (|CallParentFunction)
  (bind _outfortplayerpawn (CallFunction|GetAvatarasValidFortPlayerPawn)
    (:ValidFortPlayerPawn
      (if (Pawn|IsLocallyControlled _outfortplayerpawn)
        (Class|FortPlayerController|CacheCurrentPrimaryBarSlottedItems (Fortnite|GetFortPlayerControllerfromActor _outfortplayerpawn))))))
